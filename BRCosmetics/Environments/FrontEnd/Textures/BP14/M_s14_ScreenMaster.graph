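# Translucent unlit master material for Season 14 front-end Battle Pass room monitors, with static switches selecting a Battle Pass logo, punch card or pulsing-node screen.
# Camera-to-object sphere masks blend each screen texture with tinted backgrounds and control the logo and card overlays by distance.
# Sine-driven UV offsets and scales animate the overlays, while faint scrolling scanline and overscan bands are added on top.
# A rotatable UV mask fades the screen edges.
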Material /BRCosmetics/Environments/FrontEnd/Textures/BP14/M_s14_ScreenMaster
graph {
 properties {blendMode=BLEND_Translucent, pixelDepthOffsetMode=PDOM_Legacy, shadingModel=MSM_Unlit, usedShadingModels=Unlit}
 nodes [
  n1 = TextureCoordinate()
  n2 = Time()
  n3 = ScalarParameter() {defaultValue=1.5, parameterName=ScanTimeOffset}
  n4 = AppendVector(A=Constant(), B=Multiply(A=Add(A=@n2, B=@n3), B=ScalarParameter() {defaultValue=-1, parameterName=scanline-time factor}))
  n5 = Multiply(A=ScalarParameter() {defaultValue=0.25, parameterName=Overscan Time Factor}, B=@n4)
  n6 = ComponentMask(None=Frac(None=Add(A=@n1, B=@n5))) {g=true}
  n7 = Constant3Vector() {constant=<rgba 1 1 1 0>}
  n8 = Constant() {r=1}
  n9 = ScalarParameter() {defaultValue=30, parameterName=y-Tiles}
  n10 = Add(A=Multiply(A=@n1, B=AppendVector(A=@n8, B=@n9)), B=@n4)
  n11 = Multiply(
   A=Add(A=Sine(None=@n10), B=ScalarParameter() {defaultValue=1.5, parameterName=ScanOffset})
   B=ScalarParameter() {defaultValue=0.75, parameterName=ScanAmplitude}
  )
  n12 = Multiply(A=@n7, B=Saturate(None=ComponentMask(None=@n11) {g=true}))
  n13 = ScalarParameter() {defaultValue=0.01, parameterName=ScanStrength}
  n14 = TextureSampleParameter2D() {parameterName=ScreenTexture, texture=&/T_S14_BPRoomMonitor_Catch}
  n15 = SphereMask(
   A=CameraPositionWS()
   B=ObjectPositionWS()
   Radius=ScalarParameter() {defaultValue=620, parameterName=Radius-BP-Alpha}
   Hardness=ScalarParameter() {defaultValue=0.5, parameterName=Falloff-Alpha}
  )
  n16 = Clamp(None=@n15) {maxDefault=0.85}
  n17 = Constant3Vector() {constant=<rgba 0.0075 0.0343 0.162 1>}
  n18 = Constant()
  n19 = Multiply(A=Time(), B=ScalarParameter() {defaultValue=0.125, parameterName=TimeFac})
  n20 = Multiply(
   A=Add(A=Sine(None=@n19), B=ScalarParameter() {defaultValue=2, parameterName=Offset})
   B=ScalarParameter() {defaultValue=0.025, parameterName=Amplitude}
  )
  n21 = MaterialFunctionCall(
   UVs=Add(A=TextureCoordinate(), B=AppendVector(A=@n18, B=@n20))
   Texture Scale=Multiply(A=Constant2Vector() {g=1, r=0.5}, B=ScalarParameter() {defaultValue=0.75, parameterName=Scale Fact})
  ) {materialFunction=&ScaleUVsByCenter}
  n22 = MaterialFunctionCall(
   UVs=@n21.UVs
   CenterPosition=Constant2Vector() {g=0.5, r=0.5}
   Radius=ScalarParameter() {defaultValue=0.75, parameterName=Radius}
   Density=ScalarParameter() {defaultValue=0.75, parameterName=Density}
  ) {materialFunction=&RadialGradientExponential}
  n23 = ScalarParameter() {defaultValue=1, parameterName=Power}
  n24 = TextureSampleParameter2D(UVs=@n21.UVs) {
   parameterName=BP_LogoOverlay
   samplerSource=SSM_Clamp_WorldGroupSettings
   texture=&/T_S14_BPRoomMonitor_GoldBP_02-Logo
  }
  n25 = OneMinus(None=@n15)
  n26 = LinearInterpolate(
   A=LinearInterpolate(
    A=LinearInterpolate(A=@n14.RGB, B=@n17, Alpha=@n16)
    B=VectorParameter() {parameterName=RGB, defaultValue=<rgba 0 0.6984 1 1>}.RGB
    Alpha=Power(Base=@n22.RadialGradientExponential, Exp=@n23)
   )
   B=@n24.RGB
   Alpha=Multiply(A=@n24.A, B=Clamp(None=@n25))
  )
  n27 = TextureSampleParameter2D() {parameterName=ScreenTexture, texture=&T_S14_BPRoomMonitor_Catch}
  n28 = SphereMask(
   A=CameraPositionWS()
   B=ObjectPositionWS()
   Radius=ScalarParameter() {defaultValue=360, parameterName=Radius-PC-Alpha}
   Hardness=ScalarParameter() {defaultValue=0.6, parameterName=Falloff-PC-Alpha}
  )
  n29 = Saturate(None=@n28)
  n30 = LinearInterpolate(
   A=ComponentMask(None=VectorParameter() {group=Background-A, parameterName=RGB1, defaultValue=<rgba 0.401 0.0068 0.071 1>}.RGB) {b=true, g=true, r=true}
   B=ComponentMask(None=VectorParameter() {group=Background-A, parameterName=RGB2, defaultValue=<rgba 0.8385 0.1955 0.1371 1>}.RGB) {b=true, g=true, r=true}
   Alpha=@n27.B
  )
  n31 = TextureCoordinate() {uTiling=2}
  n32 = Constant2Vector() {r=-0.5}
  n33 = Time()
  n34 = ScalarParameter() {defaultValue=0.25, parameterName=TimeFactor-Cards}
  n35 = Add(A=Sine(None=Multiply(A=@n33, B=@n34))) {constB=25}
  n36 = TextureSample(UVs=MaterialFunctionCall(UVs=Add(A=@n31, B=@n32), Texture Scale=Multiply(A=@n35) {constB=0.03}) {materialFunction=&ScaleUVsByCenter}.UVs) {samplerSource=SSM_Clamp_WorldGroupSettings, texture=&/T_S14_BPRoomMonitor_PunchCards-Card}
  n37 = Clamp(None=OneMinus(None=@n28))
  n38 = TextureSample() {texture=&T_S14_BPRoomMonitor_Catch}
  n39 = Constant3Vector() {constant=<rgba 0.0075 0.0343 0.162 1>}
  n40 = SphereMask(
   A=CameraPositionWS()
   B=ObjectPositionWS()
   Radius=ScalarParameter() {defaultValue=360, parameterName=Radius-PC-Alpha}
   Hardness=ScalarParameter() {defaultValue=0.5, parameterName=Falloff-Alpha}
  )
  n41 = Clamp(None=@n40)
  n42 = Sine(None=Multiply(A=Time(), B=ScalarParameter() {defaultValue=0.45, parameterName=TimeFactor-Nodes})) {period=2}
  n43 = TextureCoordinate() {uTiling=2}
  n44 = Constant2Vector() {g=0.15, r=-0.75}
  n45 = ScalarParameter() {defaultValue=-0.125, parameterName=SineOffSet}
  n46 = Add(A=@n42, B=@n45) {constB=0.25}
  n47 = ScalarParameter() {parameterName=MIn}
  n48 = ScalarParameter() {defaultValue=0.55, parameterName=Max}
  n49 = TextureObject() {texture=&/T_S14_BPRoomMonitor_Catch-Elements}
  n50 = TextureSample(
   UVs=MaterialFunctionCall(UVs=Add(A=@n43, B=@n44), Texture Scale=Clamp(None=@n46, Min=@n47, Max=@n48) {maxDefault=0.25}) {materialFunction=&ScaleUVsByCenter}.UVs
   Tex=@n49
  ) {samplerSource=SSM_Clamp_WorldGroupSettings, texture=&T_S14_BPRoomMonitor_Catch-Elements}
  n51 = Subtract(A=Multiply(A=Clamp(None=@n42), B=@n50.R)) {constB=0.5}
  n52 = ScalarParameter() {defaultValue=2, parameterName=Multiply}
  n53 = TextureCoordinate() {uTiling=2}
  n54 = Constant2Vector() {g=-0.1, r=-0.5}
  n55 = Time()
  n56 = ScalarParameter() {defaultValue=2, parameterName=Offset}
  n57 = Sine(None=Multiply(A=Add(A=@n55, B=@n56) {constB=2}, B=ScalarParameter() {defaultValue=0.45, parameterName=TimeFactor-Nodes})) {period=2}
  n58 = Add(A=@n57, B=@n45) {constB=0.25}
  n59 = TextureSample(
   UVs=MaterialFunctionCall(UVs=Add(A=@n53, B=@n54), Texture Scale=Clamp(None=@n58, Min=@n47, Max=@n48) {maxDefault=0.25}) {materialFunction=&ScaleUVsByCenter}.UVs
   Tex=@n49
  ) {samplerSource=SSM_Clamp_WorldGroupSettings, texture=&T_S14_BPRoomMonitor_Catch-Elements}
  n60 = Subtract(A=Multiply(A=@n59.R, B=Clamp(None=@n57))) {constB=0.5}
  n61 = Add(A=Saturate(None=Multiply(A=@n51, B=@n52) {constB=4}), B=Saturate(None=Multiply(A=@n60, B=@n52) {constB=4}))
  n62 = StaticSwitchParameter(
   True=LinearInterpolate(A=LinearInterpolate(A=@n27.RGB, B=@n30, Alpha=@n29), B=@n36.RGB, Alpha=Multiply(A=@n36.A, B=@n37))
   False=LinearInterpolate(A=LinearInterpolate(A=@n38.RGB, B=@n39, Alpha=@n41), B=Constant3Vector() {constant=<rgba 1 0.25 0.25 0>}, Alpha=Saturate(None=@n61))
  ) {parameterName=bIsPunchCard}
  Add(
   A=Multiply(A=Multiply(A=@n6, B=@n7), B=ScalarParameter() {defaultValue=0.01, parameterName=OverScan})
   B=Add(A=Multiply(A=@n12, B=@n13), B=StaticSwitchParameter(True=@n26, False=@n62) {parameterName=bIsBP})
  )
  OneMinus(None=@n40)
  n63 = TextureCoordinate()
  n64 = ScalarParameter() {parameterName=Rotation}
  n65 = Abs(None=Add(A=Rotator(Coordinate=@n63, Time=@n64) {constCoordinate=6}) {constB=-0.5})
  n66 = ComponentMask(None=@n65) {r=true}
  n67 = ScalarParameter() {defaultValue=0.465, parameterName=x edge offset}
  n68 = Saturate(None=Multiply(A=Subtract(A=@n66, B=@n67), B=ScalarParameter() {defaultValue=16, parameterName=edge falloff}))
  n69 = ComponentMask(None=@n65) {g=true}
  n70 = ScalarParameter() {defaultValue=0.45, parameterName=y edge offset}
  n71 = Saturate(None=Multiply(A=Subtract(A=@n69, B=@n70), B=ScalarParameter() {defaultValue=16, parameterName=edge falloff}))
  Saturate(None=OneMinus(None=Add(A=@n68, B=@n71)))
  ScalarParameter() {parameterName=PC_BG_Color_Mask}
  ScalarParameter() {parameterName=BP_BG_Fade}
  ScalarParameter() {parameterName=WC_BG_Fade}
  ScalarParameter() {defaultValue=1, parameterName=BP-Alpha}
  ScalarParameter() {defaultValue=1, parameterName=Card-Alpha}
 ]
}
